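# Vehicle cosmetic material for the SwooshFC wheel FX, showing an animated, blocky three-color glow on a black base.
# Tiled, floored UVs form cells whose time-offset sine and cosine waves blend Color2 and Color3, while a second, fixed-rate wave mixes in Color1.
# The result drives Emissive through material attributes, then passes through the shared vehicle damage and de-rez transition functions.
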Material /VehicleCosmetics/Wheels/SwooshFC/Materials/M_Wheel_SwooshFC_FX
graph {
 properties {bAutomaticallySetUsageInEditor=false, bUseMaterialAttributes=true, bUsedWithSkeletalMesh=true, usedShadingModels=Default Lit}
 nodes [
  n1 = Time()
  n2 = Multiply(A=@n1, B=ScalarParameter() {defaultValue=0.5, parameterName=Wawe_Speed})
  n3 = TextureCoordinate() {uTiling=3.333, vTiling=3.333}
  n4 = Multiply(A=ScalarParameter() {defaultValue=3, parameterName=Tile}, B=ScalarParameter() {defaultValue=1, parameterName=WaweTile})
  n5 = Floor(None=Floor(None=Multiply(A=@n3, B=@n4)))
  n6 = Add(A=AppendVector(A=@n2, B=@n2), B=@n5)
  n7 = Add(A=Desaturation(None=Sine(None=@n6) {desc=6, period=10}) {luminanceFactors=<rgba 0.3 0.59 0 0>}, B=@n2)
  n8 = Constant() {r=1.25}
  n9 = Cosine(None=@n6) {period=6}
  n10 = Sine(None=Add(A=Multiply(A=@n7, B=@n8), B=ComponentMask(None=@n9) {g=true})) {desc=35, period=3}
  n11 = Multiply(A=@n1, B=Constant() {r=2})
  n12 = AppendVector(A=@n11, B=@n11)
  n13 = Desaturation(None=Sine(None=Add(A=@n12, B=@n5)) {desc=6, period=10}) {luminanceFactors=<rgba 0.3 0.59 0 0>}
  n14 = OneMinus(None=@n13)
  n15 = LinearInterpolate(
   A=LinearInterpolate(
    A=VectorParameter() {parameterName=Color2, defaultValue=<rgba 0.0018 0.1301 1 1>}.RGB
    B=VectorParameter() {parameterName=Color3, defaultValue=<rgba 1 0.0168 0.4452 1>}.RGB
    Alpha=Clamp(None=@n10)
   )
   B=VectorParameter() {parameterName=Color1, defaultValue=<rgba 0.0242 1 0.0212 1>}.RGB
   Alpha=Multiply(A=ConstantBiasScale(None=@n14) {bias=-0.96, scale=0.3}, B=@n13)
  )
  n16 = Constant()
  MaterialFunctionCall(In=MaterialFunctionCall(Material=SetMaterialAttributes!(MaterialAttributes, Emissive Color=@n15, Base Color=@n16) {attributeSetTypes [#1, #2]}) {materialFunction=&MF_BaseVehicle_Damage}.Result) {materialFunction=&MF_VehicleDeRez_Transition}
 ]
}
