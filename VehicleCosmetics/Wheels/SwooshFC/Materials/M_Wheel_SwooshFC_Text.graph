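# Translucent material for the logo text on SwooshFC vehicle wheels, drawing the logo with a fake drop shadow.
# The logo's red channel blends Shadow_Light toward white for base color and, scaled by EmissiveGlow, emissive.
# Opacity takes the maximum of that channel and the green channel sampled at offset, bump-offset UVs, scaled by Shadow_Intensity.
# The attributes then pass through the shared vehicle damage and de-rez transition functions.
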Material /VehicleCosmetics/Wheels/SwooshFC/Materials/M_Wheel_SwooshFC_Text
graph {
 properties {
  bAutomaticallySetUsageInEditor=false
  bUseMaterialAttributes=true
  bUsedWithSkeletalMesh=true
  blendMode=BLEND_Translucent
  usedShadingModels=Default Lit
 }
 nodes [
  n1 = TextureCoordinate()
  n2 = TextureSample(UVs=@n1) {samplerSource=SSM_Wrap_WorldGroupSettings, texture=/VehicleCosmetics/Wheels/SwooshFC/Textures/T_SwooshFC_Logo}
  n3 = LinearInterpolate(A=ScalarParameter() {parameterName=Shadow_Light}, B=Constant() {r=1}, Alpha=@n2.R)
  n4 = Multiply(A=@n3, B=ScalarParameter() {defaultValue=1, parameterName=EmissiveGlow})
  n5 = Add(A=@n1, B=Constant2Vector() {g=-0.03, r=-0.03})
  n6 = Constant()
  n7 = Multiply(
   A=TextureSample(UVs=BumpOffset(Coordinate=@n5, Height=@n6)) {samplerSource=SSM_Wrap_WorldGroupSettings, texture=&T_SwooshFC_Logo}.G
   B=ScalarParameter() {defaultValue=1, parameterName=Shadow_Intensity}
  )
  n8 = Max(A=@n2.R, B=Clamp(None=ConstantBiasScale(None=@n7) {bias=0, scale=1}))
  MaterialFunctionCall(In=MaterialFunctionCall(Material=SetMaterialAttributes!(MaterialAttributes, Emissive Color=@n4, Opacity=@n8, Base Color=@n3) {attributeSetTypes [#1, #2, #3]}) {materialFunction=&MF_BaseVehicle_Damage}.Result) {materialFunction=&MF_VehicleDeRez_Transition}
  Multiply(A=@n2.R)
 ]
}
